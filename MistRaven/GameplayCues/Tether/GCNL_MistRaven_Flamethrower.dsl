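; Looping gameplay cue that plays the MistRaven raven flamethrower's visuals and audio on a Fort pawn.
; Loop start spawns tuned Niagara thrower and hand effects, first/third-person audio loops, and a sound indicator for non-local pawns.
; A FakeTick timer aligns the thrower to base aim and feeds magazine ammo into audio and hand VFX.
; Another timer replays travel audio at random forward points; removal fades loops and deactivates the thrower.
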(event EventOnApplication(Cascade)
  (params (MyTarget :type Actor%20Object%20Reference) (Parameters :type Gameplay%20Cue%20Parameters%20Structure%20%28by%20ref%29) (ParticleComponents :type Array%20of%20Cascade%20Particle%20System%20Component%20Object%20References) (AudioComponents :type Array%20of%20Audio%20Component%20Object%20References) (BurstCameraShakeInstance :type Camera%20Shake%20Base%20Object%20Reference) (BurstDecalInstance :type Decal%20Actor%20Object%20Reference)))

(event Custom|Aim)

(event EventOnLoopingStart(EitherVFX-preferthisversion)
  (params (MyTarget :type Actor%20Object%20Reference) (Parameters :type Gameplay%20Cue%20Parameters%20Structure%20%28by%20ref%29) (ParticleComponents :type Array%20of%20FXSystem%20Component%20Object%20References) (AudioComponents :type Array%20of%20Audio%20Component%20Object%20References))
  (bind _as_fort_pawn (Variables|Default|GetAs%20Fort%20Pawn))
  (bind _returnvalue (Weapon|GetCurrentWeapon _as_fort_pawn))
  (bind _returnvalue_2 (FortWeapon|GetLocalMagazineAmmoCount _returnvalue))
  (bind _returnvalue_3 (Math|Conversions|ToFloat(Integer) _returnvalue_2))
  (bind _as_fort_pawn_2 (Variables|Default|GetAs%20Fort%20Pawn))
  (bind _mesh (Variables|/Script/Engine.Character|GetMesh _as_fort_pawn_2))
  (bind _asfort_pawn (Utilities|Casting|CastToFortPawn MyTarget))
  (Variables|Default|SetAs%20Fort%20Pawn _asfort_pawn)
  (Utilities|Time|SetTimerbyEvent (EventDispatchers|CreateEvent :!event PlayProjectileTravelAudio1) 0.25 true)
  (if (not (Pawn|IsLocallyControlled (Variables|Default|GetAs%20Fort%20Pawn)))
    (SoundIndicator|StartTargetedSoundIndicatorLooping (Variables|Default|GetFortSoundIndicator) (Variables|Default|GetAs%20Fort%20Pawn) "/MistRaven/Assets/Audio/Temp_AudioVisualizer/MistRaven_Temp_Fire_Cue.MistRaven_Temp_Fire_Cue" 0.1 (Variables|/Script/Engine.Pawn|GetPlayerState (Variables|Default|GetAs%20Fort%20Pawn))))
  (bind _returnvalue_4 (Niagara|SpawnSystemAttached "/MistRaven/VFX/NS_MistRaven_DarkCoil_Hands.NS_MistRaven_DarkCoil_Hands" _mesh "hand_attach_r" "0, 0, 0" "0, 0, 0" "KeepRelativeOffset" true))
  (bind _output_get (Variables|Default|SetSpawnedEffect_hand _returnvalue_4))
  (CallFunction|SpawnAudio _output_get)
  (bind _returnvalue_5 (Niagara|SpawnSystemAttached "/MistRaven/VFX/NS_MistRaven_DarkCoil_RavenThrower_V3b.NS_MistRaven_DarkCoil_RavenThrower_V3b" _mesh "root" (Variables|Default|GetNiagara%20Offset%20Location) "0, 0, 0" "KeepRelativeOffset" true))
  (Variables|Default|SetSpawnedEffect _returnvalue_5)
  (Effects|Components|ParticleSystem|SetFloatParameter _returnvalue_5 "Spawn Rate" (ScalableFloat|GetValueatLevel (Variables|Default|GetSpawnRate)))
  (Effects|Components|ParticleSystem|SetFloatParameter _returnvalue_5 "Velocity Scale" (ScalableFloat|GetValueatLevel (Variables|Default|GetVelocityRate)))
  (Effects|Components|ParticleSystem|SetFloatParameter _returnvalue_5 "Cone Angle" (ScalableFloat|GetValueatLevel (Variables|Default|GetConeAngle)))
  (Effects|Components|ParticleSystem|SetFloatParameter _returnvalue_5 "Raven Kill Distance" 1.5e+03)
  (Effects|Components|ParticleSystem|SetVectorParameter _returnvalue_5 "TargetVector" (Variables|Default|GetNiagara%20Thrower%20Rotation%20Adjustment))
  (bind _returnvalue_6 (GameplayTags|HasMatchingGameplayTag(Message) (Variables|Default|GetAs%20Fort%20Pawn) "(TagName=\"MistRaven.AllAscendedEquipped\")"))
  (Effects|Components|ParticleSystem|SetBoolParameter _returnvalue_5 "Ascended" _returnvalue_6)
  (Utilities|Time|SetTimerbyEvent (delegate Custom|FakeTick) 0.05 true true)
  (if (Pawn|IsLocallyControlled (Variables|Default|GetAs%20Fort%20Pawn))
    (Fortnite|GetPlayerAimPoint (Fortnite|GetFortPlayerControllerfromActor (Variables|Default|GetAs%20Fort%20Pawn)) 1.5e+04 (Utilities|Array|MakeArray (Variables|Default|GetAs%20Fort%20Pawn))))
  (Transformation|SetWorldRotation (Variables|Default|GetSpawnedEffect) (Pawn|GetBaseAimRotation _as_fort_pawn) false false)
  (Audio|Parameter|SetFloatParameter (Variables|Default|GetEnergyLoopAudio) "AmmoInMagazine" (Math|Curves|GetFloatValue (Variables|Default|GetAmmoAudioRampingCurve) (/ _returnvalue_3 2e+01)) :!class /Script/AudioExtensions.AudioParameterControllerInterface)
  (Niagara|SetNiagaraVariable(Float) (Variables|Default|GetSpawnedEffect_hand) "Ammo" (Math|Float|Min(Float) (/ _returnvalue_3 125.0) 1.0)))

(event Custom|FakeTick
  (bind _as_fort_pawn (Variables|Default|GetAs%20Fort%20Pawn))
  (bind _returnvalue (Weapon|GetCurrentWeapon _as_fort_pawn))
  (bind _returnvalue_2 (FortWeapon|GetLocalMagazineAmmoCount _returnvalue))
  (bind _returnvalue_3 (Math|Conversions|ToFloat(Integer) _returnvalue_2))
  (if (Pawn|IsLocallyControlled (Variables|Default|GetAs%20Fort%20Pawn))
    (Fortnite|GetPlayerAimPoint (Fortnite|GetFortPlayerControllerfromActor (Variables|Default|GetAs%20Fort%20Pawn)) 1.5e+04 (Utilities|Array|MakeArray (Variables|Default|GetAs%20Fort%20Pawn))))
  (Transformation|SetWorldRotation (Variables|Default|GetSpawnedEffect) (Pawn|GetBaseAimRotation _as_fort_pawn) false false)
  (Audio|Parameter|SetFloatParameter (Variables|Default|GetEnergyLoopAudio) "AmmoInMagazine" (Math|Curves|GetFloatValue (Variables|Default|GetAmmoAudioRampingCurve) (/ _returnvalue_3 2e+01)) :!class /Script/AudioExtensions.AudioParameterControllerInterface)
  (Niagara|SetNiagaraVariable(Float) (Variables|Default|GetSpawnedEffect_hand) "Ammo" (Math|Float|Min(Float) (/ _returnvalue_3 125.0) 1.0)))

(event EventOnRemoval(EitherVFX-preferthisversion)
  (params (MyTarget :type Actor%20Object%20Reference) (Parameters :type Gameplay%20Cue%20Parameters%20Structure%20%28by%20ref%29) (ParticleComponents :type Array%20of%20FXSystem%20Component%20Object%20References) (AudioComponents :type Array%20of%20Audio%20Component%20Object%20References) (BurstCameraShakeInstance :type Camera%20Shake%20Base%20Object%20Reference) (BurstDecalInstance :type Decal%20Actor%20Object%20Reference))
  (CallFunction|FadeOutAudio)
  (Components|Activation|Deactivate (Variables|Default|GetSpawnedEffect))
  (SoundIndicator|StopSoundIndicatorLooping (Variables|Default|GetFortSoundIndicator)))

(event EventOnRemoval(Cascade)
  (params (MyTarget :type Actor%20Object%20Reference) (Parameters :type Gameplay%20Cue%20Parameters%20Structure%20%28by%20ref%29) (ParticleComponents :type Array%20of%20Cascade%20Particle%20System%20Component%20Object%20References) (AudioComponents :type Array%20of%20Audio%20Component%20Object%20References) (BurstCameraShakeInstance :type Camera%20Shake%20Base%20Object%20Reference) (BurstDecalInstance :type Decal%20Actor%20Object%20Reference)))

(event EventOnRecurring(Cascade)
  (params (MyTarget :type Actor%20Object%20Reference) (Parameters :type Gameplay%20Cue%20Parameters%20Structure%20%28by%20ref%29) (ParticleComponents :type Array%20of%20Cascade%20Particle%20System%20Component%20Object%20References) (AudioComponents :type Array%20of%20Audio%20Component%20Object%20References) (BurstCameraShakeInstance :type Camera%20Shake%20Base%20Object%20Reference) (BurstDecalInstance :type Decal%20Actor%20Object%20Reference)))

(fn FadeOutAudio
  (Audio|Components|Audio|FadeOut (Variables|Default|GetHandAudio) 0.5 0.0 "Linear" :!class /Script/Engine.AudioComponent)
  (Audio|Components|Audio|FadeOut (Variables|Default|GetEnergyLoopAudio) 0.2 0.0 "Linear" :!class /Script/Engine.AudioComponent))

(fn PlayProjectileTravelAudio1
  (bind _randomspawnvector (LocalVariables|GetRandomSpawnVector))
  (LocalVariables|SetRandomSpawnVector (+ (Transformation|GetActorLocation (Variables|Default|GetAs%20Fort%20Pawn)) (* (Math|Random|RandomFloatinRange 8e+02 1.5e+03) (Math|Vector|GetForwardVector (Math|Rotator|MakeRotator (+ (.roll (Math|Rotator|MakeRotfromX (Transformation|GetActorForwardVector (Variables|Default|GetAs%20Fort%20Pawn)))) (Math|Random|RandomFloatinRange -5.0 5.0)) (+ (.pitch (Math|Rotator|MakeRotfromX (Transformation|GetActorForwardVector (Variables|Default|GetAs%20Fort%20Pawn)))) (Math|Random|RandomFloatinRange -5.0 5.0)) (+ (.yaw (Math|Rotator|MakeRotfromX (Transformation|GetActorForwardVector (Variables|Default|GetAs%20Fort%20Pawn)))) (Math|Random|RandomFloatinRange -5.0 5.0)))))))
  (bind _projectilespawnedaudio (Variables|Default|GetProjectileSpawnedAudio)
    (:then
      (Transformation|SetWorldLocation _projectilespawnedaudio _randomspawnvector false false)
      (Audio|Parameter|ExecuteTriggerParameter _projectilespawnedaudio "ReplayTravelSound"))
    (:else
      (bind _returnvalue (Audio|SpawnSoundatLocation 0 (select (ScalableFloat|GetValueasBool (Variables|Default|GetMistRavenPolyphonyHotfix)) "/MistRaven/Assets/Audio/MSS_MistRaven_Flamethrower_Projectile_Travel_V2.MSS_MistRaven_Flamethrower_Projectile_Travel_V2" "/MistRaven/Assets/Audio/MSS_MistRaven_Flamethrower_Projectile_Travel_Polyphony.MSS_MistRaven_Flamethrower_Projectile_Travel_Polyphony") _randomspawnvector))
      (Variables|Default|SetProjectileSpawnedAudio _returnvalue))))

(fn SpawnAudio
  (params (AttachToComponent :type Scene%20Component%20Object%20Reference))
  (bind _returnvalue (Audio|SpawnSoundAttached (select (Pawn|IsLocallyViewed (Variables|Default|GetAs%20Fort%20Pawn)) "/MistRaven/Assets/Audio/MSS_MistRaven_Flamethrower_CrowLoop_1P.MSS_MistRaven_Flamethrower_CrowLoop_1P" "/MistRaven/Assets/Audio/MSS_MistRaven_Flamethrower_CrowLoop_3P.MSS_MistRaven_Flamethrower_CrowLoop_3P") AttachToComponent))
  (Variables|Default|SetHandAudio _returnvalue)
  (bind _returnvalue_2 (Audio|SpawnSoundAttached (select (Pawn|IsLocallyViewed (Variables|Default|GetAs%20Fort%20Pawn)) "/MistRaven/Assets/Audio/MSS_MistRaven_Flamethrower_EnergyLoop_1P.MSS_MistRaven_Flamethrower_EnergyLoop_1P" "/MistRaven/Assets/Audio/MSS_MistRaven_Flamethrower_EnergyLoop_3P.MSS_MistRaven_Flamethrower_EnergyLoop_3P") (Variables|/Script/Engine.Actor|GetRootComponent (Variables|Default|GetAs%20Fort%20Pawn))))
  (Variables|Default|SetEnergyLoopAudio _returnvalue_2))

(fn ConstructionScript)
